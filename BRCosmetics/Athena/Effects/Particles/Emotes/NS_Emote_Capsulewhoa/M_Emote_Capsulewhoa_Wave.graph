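# Additive unlit Niagara mesh-particle material for the wave effect in the Capsulewhoa emote.
# A panning blurred-noise sample, raised to a power, drives tinted emissive through MF_FX_EmissiveAdjustments, a refraction-style lerp and normal-direction displacement.
# Particle color tints the emissive, and particle alpha scales both the Fresnel term and the refraction blend.
# Tiling and panning reuse their U parameter on both axes, leaving the V channels unused.
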Material /BRCosmetics/Athena/Effects/Particles/Emotes/NS_Emote_Capsulewhoa/M_Emote_Capsulewhoa_Wave
graph {
 properties {
  bUsedWithNiagaraMeshParticles=true
  blendMode=BLEND_Additive
  pixelDepthOffsetMode=PDOM_Legacy
  refractionMethod=RM_IndexOfRefraction
  shadingModel=MSM_Unlit
  usedShadingModels=Unlit
 }
 nodes [
  n1 = ParticleColor()
  n2 = VectorParameter() {parameterName=Emissive UV, channelNames {a=V_Pan, b=U_Pan, g=V_Tile, r=U_Tile}, defaultValue=<rgba 0.5 1 0.5 0.2>}
  n3 = Multiply(A=TextureCoordinate(), B=AppendVector(A=@n2.U_Tile, B=@n2.U_Tile))
  n4 = AppendVector(A=@n2.U_Pan, B=@n2.U_Pan)
  n5 = Power(
   Base=TextureSample(UVs=Panner(Coordinate=@n3, Speed=@n4)) {samplerType=SAMPLERTYPE_Grayscale, texture=&LowResBlurredNoise}.G
   Exp=ScalarParameter() {defaultValue=0.5, parameterName=EmissiveTexturePower}
  )
  n6 = ScalarParameter() {defaultValue=2, parameterName=Refaction Intensity}
  LinearInterpolate(Alpha=Multiply(A=@n1.A, B=Multiply(A=@n5, B=@n6))) {constA=1, constB=1.05}
  n7 = VectorParameter() {parameterName=Emissive Color, defaultValue=<rgba 0.55 1.1 0.99 1>}
  MaterialFunctionCall(Emissive=Multiply(A=Multiply(A=@n5, B=@n7.RGB), B=@n1.RGB)) {materialFunction=&MF_FX_EmissiveAdjustments}
  Multiply(
   A=Fresnel(
    ExponentIn=ScalarParameter() {defaultValue=3, parameterName=Fresnel ExponentIn}
    BaseReflectFractionIn=ScalarParameter() {defaultValue=0.01, parameterName=Fresnel Base}
   )
   B=@n1.A
  )
  Multiply(A=Multiply(A=@n5, B=ScalarParameter() {defaultValue=0.2, parameterName=WPO Intensity}) {constB=2}, B=VertexNormalWS())
 ]
}
