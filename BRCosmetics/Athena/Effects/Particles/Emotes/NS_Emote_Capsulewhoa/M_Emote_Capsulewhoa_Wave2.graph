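# Translucent unlit Niagara mesh-particle material for the Capsulewhoa emote wave, drawing a chromatic refraction of the scene.
# Per-channel refracted offsets sample a downsampled scene-color render target, projected from world to screen, or offset in view space when the front-end MPC flag is set.
# Target alpha, treated as depth, drives refraction scale, tint, scatter and brightness, alongside Fresnel coloring that spares a target hue.
# Camera distance and particle alpha fade it; Forward and Mobile paths output zero.
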Material /BRCosmetics/Athena/Effects/Particles/Emotes/NS_Emote_Capsulewhoa/M_Emote_Capsulewhoa_Wave2
graph {
 properties {
  bUsedWithNiagaraMeshParticles=true
  blendMode=BLEND_Translucent
  pixelDepthOffsetMode=PDOM_Legacy
  shadingModel=MSM_Unlit
  translucencyPass=MTP_BeforeDOF
  usedShadingModels=Unlit
 }
 nodes [
  n1 = MaterialFunctionCall!(Ray Direction=Multiply(A=CameraVectorWS()) {constB=-1}, Surface Normal=PixelNormalWS(), Refractive Index Origin, Refractive Index Target) {materialFunction=/Engine/Functions/Engine_MaterialFunctions01/Vectors/Refract}
  n2 = ScalarParameter() {defaultValue=200, parameterName=Refraction_RChannel}
  n3 = Subtract(A=CameraPositionWS(), B=WorldPosition().XYZ)
  n4 = Saturate(None=Divide(
    A=MaterialFunctionCall(Vector 3=@n3) {materialFunction=/Engine/Functions/Engine_MaterialFunctions02/Utility/VectorLength}.V3 Length
    B=ScalarParameter() {parameterName=FadeDistance}
   ) {constB=500})
  n5 = VectorParameter() {parameterName=ScaleRefByDepth, defaultValue=<rgba 0 0 0 1>}
  n6 = ScalarParameter() {parameterName=MipLevelBias}
  n7 = TextureSample!(UVs=ScreenPosition().ViewportUV, Tex, Bias=@n6, Apply View MipBias) {
   mipValueMode=TMVM_MipBias
   samplerType=SAMPLERTYPE_LinearColor
   texture=/BRCosmetics/Athena/Effects/Particles/Emotes/NS_Emote_Capsulewhoa/DownsampleSceneColor/RT_SceneColor
  }
  n8 = OneMinus(None=@n7.A)
  n9 = ScalarParameter() {defaultValue=2, parameterName=ScaleRefByDepth_DepthPower}
  n10 = LinearInterpolate(A=@n5.R, B=@n5.R, Alpha=OneMinus(None=Power(Base=@n8, Exp=@n9)))
  n11 = Multiply(A=@n1.Vector, B=Multiply(A=Multiply(A=@n2, B=@n4), B=@n10)) {constB=300}
  n12 = WorldPosition()
  n13 = MaterialFunctionCall!(
   Camera Relative Position=Subtract(A=Add(A=@n11, B=@n12.XYZ), B=CameraPositionWS())
   X Camera to World Vector
   Y Camera to World Vector
   Z Camera to World Vector
   "Tan(FOV/2)*[1,Screen Res Y/X]"
  ) {materialFunction=/Engine/Functions/Engine_MaterialFunctions02/Math/ManualWorldToScreenUVsTransform}
  n14 = Multiply(
   A=MaterialFunctionCall(Vector 3=@n11) {materialFunction=&VectorLength}.V3 Length
   B=Transform(None=Normalize(VectorInput=@n11)) {transformSourceType=TRANSFORMSOURCE_World, transformType=TRANSFORM_View}
  )
  n15 = ScalarParameter() {defaultValue=0.0015, parameterName=FrontEndRefraction_Strength}
  n16 = Add(A=Multiply(A=ComponentMask(None=@n14) {g=true, r=true}, B=@n15), B=ScreenPosition().ViewportUV)
  n17 = CollectionParameter() {collection=/BRCosmetics/Effects/FrontEndDetecter/MPC_FrontEndDetecter, parameterName=bIsFrontEnd}
  n18 = TextureSample(UVs=Saturate(None=LinearInterpolate(A=@n13.Screen UVs, B=@n16, Alpha=@n17)), Bias=@n6) {
   mipValueMode=TMVM_MipBias
   samplerType=SAMPLERTYPE_LinearColor
   texture=&RT_SceneColor
  }
  n19 = ScalarParameter() {defaultValue=200, parameterName=Refraction_GChannel}
  n20 = Multiply(A=@n1.Vector, B=Multiply(A=Multiply(A=@n4, B=@n19), B=@n10)) {constB=200}
  n21 = WorldPosition()
  n22 = MaterialFunctionCall(Camera Relative Position=Subtract(A=Add(A=@n20, B=@n21.XYZ), B=CameraPositionWS())) {materialFunction=&ManualWorldToScreenUVsTransform}
  n23 = Multiply(
   A=MaterialFunctionCall(Vector 3=@n20) {materialFunction=&VectorLength}.V3 Length
   B=Transform(None=Normalize(VectorInput=@n20)) {transformSourceType=TRANSFORMSOURCE_World, transformType=TRANSFORM_View}
  )
  n24 = Add(A=Multiply(A=ComponentMask(None=@n23) {g=true, r=true}, B=@n15), B=ScreenPosition().ViewportUV)
  n25 = TextureSample(UVs=Saturate(None=LinearInterpolate(A=@n22.Screen UVs, B=@n24, Alpha=@n17)), Bias=@n6) {
   mipValueMode=TMVM_MipBias
   samplerType=SAMPLERTYPE_LinearColor
   texture=&RT_SceneColor
  }
  n26 = ScalarParameter() {defaultValue=200, parameterName=Refraction_BChannel}
  n27 = Multiply(A=@n1.Vector, B=Multiply(A=Multiply(A=@n4, B=@n26), B=@n10)) {constB=100}
  n28 = WorldPosition()
  n29 = MaterialFunctionCall(Camera Relative Position=Subtract(A=Add(A=@n27, B=@n28.XYZ), B=CameraPositionWS())) {materialFunction=&ManualWorldToScreenUVsTransform}
  n30 = Multiply(
   A=MaterialFunctionCall(Vector 3=@n27) {materialFunction=&VectorLength}.V3 Length
   B=Transform(None=Normalize(VectorInput=@n27)) {transformSourceType=TRANSFORMSOURCE_World, transformType=TRANSFORM_View}
  )
  n31 = Add(A=Multiply(A=ComponentMask(None=@n30) {g=true, r=true}, B=@n15), B=ScreenPosition().ViewportUV)
  n32 = TextureSample(UVs=Saturate(None=LinearInterpolate(A=@n29.Screen UVs, B=@n31, Alpha=@n17)), Bias=@n6) {
   mipValueMode=TMVM_MipBias
   samplerType=SAMPLERTYPE_LinearColor
   texture=&RT_SceneColor
  }
  n33 = AppendVector(A=AppendVector(A=@n18.R, B=@n25.G), B=@n32.B)
  n34 = LinearInterpolate(
   A=VectorParameter() {parameterName=FresnelColor_Inner, defaultValue=<rgba 0 0 0 1>}.RGB
   B=VectorParameter() {parameterName=FresnelColor_Outter, defaultValue=<rgba 0 0 0 1>}.RGB
   Alpha=Fresnel(
    ExponentIn=ScalarParameter() {defaultValue=4, parameterName=Fresnel_Exponent}
    BaseReflectFractionIn=ScalarParameter() {parameterName=Fresnel_BaseReflectFractionIn}
   )
  ) {constA=1}
  n35 = DotProduct(A=Normalize(VectorInput=@n33), B=VectorParameter() {parameterName=ScaleGreenColor_TargetHue, defaultValue=<rgba 0 1 0 1>}.RGB)
  n36 = ScalarParameter() {parameterName=ScaleGreenColor_RangeOffset}
  n37 = Multiply(A=Saturate(None=Add(A=@n35, B=@n36)), B=ScalarParameter() {defaultValue=1, parameterName=ScaleGreenColor_RangeStrength})
  n38 = Power(Base=@n7.A, Exp=ScalarParameter() {defaultValue=2, parameterName=AddColorByDepth_DepthPower})
  n39 = LinearInterpolate(
   A=VectorParameter() {parameterName=AddColorByDepth_Color, defaultValue=<rgba 0 0 0 1>}.RGB
   B=VectorParameter() {parameterName=AddColorByDepth_FrontEndColor, defaultValue=<rgba 0 0 0 1>}.RGB
   Alpha=@n17
  )
  n40 = VectorParameter() {parameterName=Scatt_Color, defaultValue=<rgba 0 0 0 1>}
  n41 = VectorParameter() {parameterName=Scatt_FrontEndColor, defaultValue=<rgba 0 0 0 1>}
  n42 = ScalarParameter() {parameterName=Scatt_DepthPower}
  n43 = ScalarParameter() {parameterName=Scatt_AreaOffset}
  n44 = Saturate(None=Add(A=Power(Base=@n25.A, Exp=@n42), B=@n43))
  n45 = Divide(B=Add(A=Constant() {r=1}, B=@n43))
  n46 = Add(
   A=LinearInterpolate(A=Multiply(A=@n33, B=@n34), B=@n33, Alpha=Saturate(None=@n37))
   B=LinearInterpolate(A=Multiply(A=@n38, B=@n39), B=LinearInterpolate(A=@n40.RGB, B=@n41.RGB, Alpha=@n17), Alpha=Multiply(A=@n44, B=@n45))
  )
  n47 = LinearInterpolate(
   B=ScalarParameter() {parameterName=SkyScale_BrightnessScale}
   Alpha=Power(Base=@n25.A, Exp=ScalarParameter() {parameterName=SkyScale_DepthPower})
  ) {constA=1}
  n48 = LinearInterpolate(
   A=VectorParameter() {parameterName=Tint_Color, defaultValue=<rgba 1 1 1 1>}.RGB
   B=VectorParameter() {parameterName=Tint_FrontEndColor, defaultValue=<rgba 1 1 1 1>}.RGB
   Alpha=@n17
  )
  n49 = Power(Base=@n25.A, Exp=ScalarParameter() {defaultValue=2, parameterName=Tint_DepthPower})
  n50 = Multiply(A=Saturate(None=Multiply(A=@n4) {constB=2}), B=ParticleColor().A)
  n51 = LinearInterpolate(A=SceneColor!(UVs).RGB, B=Multiply(A=Multiply(A=@n46, B=@n47), B=LinearInterpolate(B=@n48, Alpha=@n49) {constA=1}), Alpha=@n50)
  n52 = Constant()
  MaterialFunctionCall(Emissive=ShadingPathSwitch(Default=@n51, Deferred=@n51, Forward=@n52, Mobile=@n52)) {materialFunction=&MF_FX_EmissiveAdjustments}
  n53 = Saturate(None=Multiply(A=@n50) {constB=2})
  ShadingPathSwitch(Default=@n53, Deferred=@n53, Forward=@n52, Mobile=@n52)
 ]
}
