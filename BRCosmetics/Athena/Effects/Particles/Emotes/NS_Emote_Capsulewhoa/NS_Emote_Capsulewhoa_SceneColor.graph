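# Niagara system for the Capsulewhoa emote, set up around a scene color render target; the name and default texture path suggest a downsampled capture.
# Its single GPU emitter binds a 256x256 RGBA8 Render Target 2D interface to User.RenderTexture, generating mips after simulation.
# Emitter update derives an Enable flag from significance index and emitter age, plus a significance-based VisibilityTag.
# A sprite renderer is present, but no step writing the texture is shown.
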NiagaraSystem /BRCosmetics/Athena/Effects/Particles/Emotes/NS_Emote_Capsulewhoa/NS_Emote_Capsulewhoa_SceneColor
dependencies {
 usedDataInterfaces [/Script/Niagara.NiagaraDataInterfaceRenderTarget2D]
 usedDynamicInputs []
 usedModules [&SystemState]
}
emitters {Empty001 {
  simTarget=GPUComputeSim
  emitterSpawnScript [/SetVariables(Emitter.ExportRT=DataInterface() {propertyValues {
      MipMapGeneration=PostSimulate
      MipMapGenerationType=Linear
      OverrideRenderTargetFilter=TF_Default
      OverrideRenderTargetFormat=RTF_RGBA8
      bInheritUserParameterSettings=false
      bOverrideFormat=true
      bPreviewRenderTarget=false
      RenderTargetUserParameter {parameter {name=User.RenderTexture, typeDefHandle {registeredTypeIndex=4}, varData []}}
      Size=<256 256>
     }}) {isSetParameters=true}]
  emitterUpdateScript [
   &EmitterState()
   /SetVariables(
    Emitter.Enable:NiagaraBool=&Boolean_AndOperation(
     Bool1=&SetBoolByIntComparison(A=$Engine.System.SignificanceIndex, B=1, Comparison Type=::NewEnumerator6)
     Bool2=&SetBoolByFloatComparison(A=$Emitter.Age, B=0.8)
    )
    Emitter.VisibilityTag:NiagaraInt32=&MakeCustomIntFromBool(
     Bool=&SetBoolByIntComparison(A=$Engine.System.SignificanceIndex, B=1, Comparison Type=::NewEnumerator6)
     True INT=1
     False INT=0
    )
   ) {isSetParameters=true}
  ]
  rendererClasses [&NiagaraSpriteRendererProperties]
  renderers [{rendererClass=&NiagaraSpriteRendererProperties}]
 }}
summary {
 systemName=NS_Emote_Capsulewhoa_SceneColor
 emitters [{emitterName=Empty001, simTarget=GPUComputeSim, rendererClasses [&NiagaraSpriteRendererProperties]}]
 userVariables [
  User.AA {type=NiagaraFloat, defaultValue=0, description=""}
  User.RenderTexture {
   type=TextureRenderTarget
   description=""
   defaultValue {
    struct=NiagaraExt_VariableValue_Object
    value {
     object=/BRCosmetics/Athena/Effects/Particles/Emotes/NS_Emote_Capsulewhoa/DownsampleSceneColor/RT_SceneColor
     objectClass=/Script/Engine.TextureRenderTarget
    }
   }
  }
 ]
}
