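# Niagara system for the Capsulewhoa emote, drawing mesh bullets with rippling trails and bullet audio.
# Two Bullet emitters ease particles along -Y by curve, play fading persistent audio and generate location events.
# Empty lines up mesh particles along that axis, using waveforms and an index-staggered age to ripple, grow and fade them in sequence.
# Wave has no visible spawn module and scales fading meshes by spawn group.
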NiagaraSystem /BRCosmetics/Athena/Effects/Particles/Emotes/NS_Emote_Capsulewhoa/NS_Emote_Capsulewhoa_V2_NewShader
dependencies {
 usedDataInterfaces [
  &NiagaraDataInterfaceCamera
  &NiagaraDataInterfaceCurve
  &NiagaraDataInterfaceColorCurve
  &NiagaraDataInterfaceAudioPlayer
 ]
 usedDynamicInputs []
 usedModules [
  &self:SystemSpawnScript.SystemScriptSource.SystemScriptGraph.NiagaraNodeAssignment_0.SetVariables_#1
  &SystemState
 ]
}
emitters {
 Bullet {
  emitterUpdateScript [
   &EmitterState(Life Cycle Mode=::NewEnumerator1, Loop Behavior=::NewEnumerator1, Loop Duration=1, UseLoopDelay=false)
   &SpawnBurst_Instantaneous(Use Spawn Probability=false, UseLoopCountLimit=false)
   &SpawnRate(SpawnRate=5, Use Spawn Probability=false) {enabled=false}
  ]
  particleSpawnScript [
   &InitializeParticle(
    Lifetime Max=2.7
    Lifetime Min=2.5
    Lifetime Mode=::NewEnumerator1
    Mesh Scale Mode=::NewEnumerator3
    UsePositionOffset=false
    Write Material Random=true
   )
   &InitialMeshOrientation(Mesh Axis To Orient=<0 1 0>, Mesh Orientation Mode=::NewEnumerator2, Use Rotation Vector=true)
   &BoxLocation(Box Size=<0 0 0>, MaskSpawnGroup=false, Module.Override Seed=false, Rotate Box=false, Surface Only=false) {enabled=false}
   /SetVariables(Particles.MeshScale2:Vector3f=<1 0.11 0.11>) {isSetParameters=true, module=SetVariables_#2}
   /Niagara/Modules/Audio/PlayPersistentAudio(
    Audio Player:NiagaraDataInterfaceAudioPlayer=$System.PlayBulletAudio
    Enable Fade In / Out=true
    Fade In Time=0
    Fade Out Time=1
    Play Audio=true
   )
   &ApplyInitialForces()
   &GenerateLocationEvent(Use Event Probability=false) {module=GenerateLocationEvent001}
  ]
  particleUpdateScript [
   &ParticleState()
   /SetVariables(Particles.Position=&Add_Vector(
     A=Linked() {linkedVariable {name=Particles.Initial.Position, type=NiagaraPosition}}
     B=&Lerp_Vector(
      A=<0 0 0>
      B=<0 -1000 0>
      Alpha=&FloatFromCurve(FloatCurve=DataInterface() {propertyValues {Curve {keys [
           {arriveTangent=2.7524, interpMode=RCIM_Cubic, leaveTangent=2.7524, tangentMode=RCTM_User, time=0, value=0.1}
           {arriveTangent=0.6, leaveTangent=0.6, time=0.1, value=0.35}
           {leaveTangent=-0.0405, time=1, value=0.7}
          ]}}})
     )
    )) {isSetParameters=true, module=SetVariables_#3}
   /SetVariables(Particles.Dynamic2:Vector4f=&MakeVector4(
     X=&FloatFromCurve(FloatCurve=DataInterface() {propertyValues {Curve {keys [<at 0 0 RCIM_Cubic>,<at 1 2 RCIM_Cubic>]}}})
     Y=0
     Z=50
     W=0.15
    )) {isSetParameters=true, module=SetVariables_#4}
   &SolveForcesAndVelocity(Clamp Velocity=false, Limit Acceleration=false)
   &GenerateLocationEvent(Delay Before Sending Events=0.25, Event Type:ENiagara_LocEventType=::NewEnumerator2, Use Event Probability=false, UseEventDelay=true) {enabled=false}
   /Niagara/Modules/Audio/UpdatePersistentAudio(Audio Player:NiagaraDataInterfaceAudioPlayer=$System.PlayBulletAudio)
  ]
  rendererClasses [&NiagaraMeshRendererProperties]
  renderers [
   {rendererClass=&NiagaraMeshRendererProperties, rendererIndex=0}
   {rendererClass=&NiagaraMeshRendererProperties, rendererIndex=1}
  ]
 }
 Bullet001 {
  emitterUpdateScript [
   &EmitterState(Life Cycle Mode=::NewEnumerator1, Loop Behavior=::NewEnumerator1, Loop Duration=1, UseLoopDelay=false)
   &SpawnBurst_Instantaneous(Use Spawn Probability=false, UseLoopCountLimit=false)
   &SpawnRate(SpawnRate=5, Use Spawn Probability=false) {enabled=false}
  ]
  particleSpawnScript [
   &InitializeParticle(
    Lifetime Max=2.7
    Lifetime Min=2.5
    Lifetime Mode=::NewEnumerator1
    Mesh Scale Mode=::NewEnumerator3
    UsePositionOffset=false
    Write Material Random=true
   )
   &InitialMeshOrientation(Mesh Axis To Orient=<0 1 0>, Mesh Orientation Mode=::NewEnumerator2, Use Rotation Vector=true)
   &BoxLocation(Box Size=<0 0 0>, MaskSpawnGroup=false, Module.Override Seed=false, Rotate Box=false, Surface Only=false) {enabled=false}
   /SetVariables(Particles.MeshScale2:Vector3f=<1 0.11 0.11>) {isSetParameters=true, module=SetVariables_#2}
   &PlayPersistentAudio(
    Audio Player:NiagaraDataInterfaceAudioPlayer=$System.PlayBulletAudio
    Enable Fade In / Out=true
    Fade In Time=0
    Fade Out Time=1
    Play Audio=true
   )
   &ApplyInitialForces()
   &GenerateLocationEvent(Use Event Probability=false) {enabled=false, module=GenerateLocationEvent001}
  ]
  particleUpdateScript [
   &ParticleState()
   /SetVariables(Particles.Position=&Add_Vector(
     A=Linked() {linkedVariable {name=Particles.Initial.Position, type=NiagaraPosition}}
     B=&Lerp_Vector(
      A=<0 0 0>
      B=<0 -1000 0>
      Alpha=&FloatFromCurve(FloatCurve=DataInterface() {propertyValues {Curve {keys [
           {arriveTangent=2.7524, interpMode=RCIM_Cubic, leaveTangent=2.7524, tangentMode=RCTM_User, time=0, value=0.1}
           {arriveTangent=0.6, leaveTangent=0.6, time=0.1, value=0.35}
           {leaveTangent=-0.0405, time=1, value=0.7}
          ]}}})
     )
    )) {isSetParameters=true, module=SetVariables_#3}
   /SetVariables(Particles.Dynamic2:Vector4f=&MakeVector4(
     X=&FloatFromCurve(FloatCurve=DataInterface() {propertyValues {Curve {keys [<at 0 0 RCIM_Cubic>,<at 1 2 RCIM_Cubic>]}}})
     Y=0
     Z=50
     W=0.15
    )) {isSetParameters=true, module=SetVariables_#4}
   &SolveForcesAndVelocity(Clamp Velocity=false, Limit Acceleration=false)
   &GenerateLocationEvent(Delay Before Sending Events=0.25, Event Type:ENiagara_LocEventType=::NewEnumerator2, Use Event Probability=false, UseEventDelay=true)
   &UpdatePersistentAudio(Audio Player:NiagaraDataInterfaceAudioPlayer=$System.PlayBulletAudio)
  ]
  rendererClasses [&NiagaraMeshRendererProperties]
  renderers [
   {rendererClass=&NiagaraMeshRendererProperties, rendererIndex=0}
   {rendererClass=&NiagaraMeshRendererProperties, rendererIndex=1}
  ]
 }
 Empty {
  emitterSpawnScript [/SetVariables(Emitter.ParticleCount=318) {isSetParameters=true}]
  emitterUpdateScript [
   &EmitterState()
   /Niagara/Modules/Spawn/Location/SpawnParticlesInGrid(X Count=0, Y Count:NiagaraInt32=$Emitter.ParticleCount, Z Count=0)
  ]
  particleSpawnScript [
   &InitializeParticle(
    Color Mode=::NewEnumerator0
    Lifetime=5
    Mesh Scale Mode=::NewEnumerator3
    Mesh Uniform Scale=0.1
    UsePositionOffset=false
    Write Material Random=true
   )
   /Niagara/Modules/Spawn/Location/V2/GridLocation(Module.Jitter Inside Grid Cells=false, Module.Override Random Seed=false, XYZ Dimensions:Vector3f=<0 1.1399 0>)
   &InitialMeshOrientation(
    Mesh Axis To Orient=<0 1 0>
    Mesh Orientation Mode=::NewEnumerator2
    Rotation=&RandomRangeVector(Minimum=<-1 0 0>, Maximum=<1 0 0>)
    Use Rotation Vector=true
   )
   /SetVariables(Particles.Position=&Add_Vector(A=Linked() {linkedVariable {name=Particles.Position, type=NiagaraPosition}}, B=<0 -498.7092 0>)) {isSetParameters=true}
  ]
  particleUpdateScript [
   &ParticleState()
   /SetVariables(Particles.New float:NiagaraFloat=&Waveform(
     Waveform Count=::NewEnumerator1
     "Waveform [1]"=::NewEnumerator6
     "[1] Input Position"=0.5
     "[1] Phase Scale"=45
     "[1] Phase"=&Add_Float(
      A=&self:ScratchDynamicInput(New int32:NiagaraInt32=$Emitter.ParticleCount)
      B=&Multiply_Float(A=$Emitter.Age, B=0.003)
     )
     "[1] Frequency"=1
     "[1] Amplitude Min/Max"=<0 1>
     "[1] Scale"=0.6
     "[1] Bias"=0.4
     "Waveform [2] Blend"=::NewEnumerator2
     "Waveform [2]"=::NewEnumerator7
     "[2] Input Position"=0.5
     "[2] Phase"=&Add_Float(
      A=&self:ScratchDynamicInput(New int32:NiagaraInt32=$Emitter.ParticleCount)
      B=&Multiply_Float(A=$Emitter.Age, B=0.003)
     )
     "[2] Phase Scale"=45
     "[2] Compound Frequency"=2
     "[2] Frequency"=1
     "[2] Amplitude Min/Max"=<0 1>
     "[2] Scale"=0.35
     "[2] Bias"=0.65
     "[2] Exponent"=1
     "[2] Pulse Duration"=0.1
     "[2] Clip Triangle Min/Max"=<0 0>
     "[2] Random Seed"=0
     "[3] Phase"=$Emitter.Age
    )) {isSetParameters=true}
   &ScaleMeshSize(Scale Factor=&MakeVector(X=2, Y=$Particles.New float, Z=$Particles.New float))
   /SetVariables(Particles.CustomNormalizeAge:NiagaraFloat=&ClampFloat(Float=&Add_Float(
      A=&Add_Float(
       A=&Divide_Float(
        A=&self:UniqueID()
        B=&MakeFloatFromInt(Int=$Emitter.ParticleCount)
       )
       B=-1
      )
      B=&Divide_Float(A=&Add_Float(A=$Emitter.Age, B=0), B=2.35)
     ))) {isSetParameters=true}
   /SetVariables(Particles.ScaleCurve:NiagaraFloat=&FloatFromCurve(
     FloatCurve=DataInterface() {propertyValues {Curve {keys [
         {arriveTangent=0.2521, interpMode=RCIM_Cubic, leaveTangent=1.0833, tangentMode=RCTM_Break, time=0, value=0}
         {arriveTangent=1.0793, interpMode=RCIM_Cubic, tangentMode=RCTM_Break, time=1, value=0.75}
        ]}}}
     CurveIndex=$Particles.CustomNormalizeAge
    )) {isSetParameters=true}
   &ScaleMeshSize(Scale Factor=&MakeVector(
     X=&FloatFromCurve(
      FloatCurve=DataInterface() {propertyValues {Curve {keys [
          {arriveTangent=6.5965, interpMode=RCIM_Cubic, leaveTangent=6.5965, tangentMode=RCTM_User, time=0, value=0}
          {arriveTangent=0.3988, interpMode=RCIM_Cubic, leaveTangent=0.3988, tangentMode=RCTM_User, time=0.1, value=0.3}
          {arriveTangent=1.4043, interpMode=RCIM_Cubic, leaveTangent=1.4043, tangentMode=RCTM_User, time=1, value=0.9}
         ]}}}
      CurveIndex=$Particles.ScaleCurve
     )
     Y=$Particles.ScaleCurve
     Z=$Particles.ScaleCurve
    )) {module=ScaleMeshSize002}
   &ScaleColor(
    Scale Alpha=&FloatFromCurve(
     FloatCurve=DataInterface() {propertyValues {Curve {keys [
         {interpMode=RCIM_Cubic, tangentMode=RCTM_User, time=0, value=0.05}
         {interpMode=RCIM_Cubic, tangentMode=RCTM_User, time=0.15, value=1}
         {arriveTangent=0.1663, interpMode=RCIM_Cubic, leaveTangent=0.1663, tangentMode=RCTM_User, time=0.7507, value=1}
         {arriveTangent=-7.981, interpMode=RCIM_Cubic, leaveTangent=-7.981, tangentMode=RCTM_User, time=1, value=0}
        ]}}}
     CurveIndex=$Particles.CustomNormalizeAge
    )
    ScaleA=true
    ScaleRGB=true
   )
   &ScaleColor(
    Scale Alpha=&FloatFromCurve(
     FloatCurve=DataInterface() {propertyValues {Curve {keys [
         {interpMode=RCIM_Cubic, tangentMode=RCTM_User, time=0, value=0.1}
         {interpMode=RCIM_Cubic, tangentMode=RCTM_User, time=0.1, value=1}
         {arriveTangent=0.1663, interpMode=RCIM_Cubic, leaveTangent=0.1663, tangentMode=RCTM_User, time=0.75, value=1}
         {arriveTangent=-1.101, interpMode=RCIM_Cubic, leaveTangent=-1.101, tangentMode=RCTM_User, time=1, value=0.15}
        ]}}}
     CurveIndex=&Divide_Float(
      A=&self:UniqueID()
      B=&MakeFloatFromInt(Int=$Emitter.ParticleCount)
     )
    )
    ScaleA=true
    ScaleRGB=true
   ) {module=ScaleColor001}
   &KillParticles(Kill Particles=&SetBoolByFloatComparison(
     A=&MakeFloatFromLinearColor(LinearColor=$Particles.Color, Channel:ENiagaraLinearColor_Channels=::NewEnumerator3)
     B=0.01
     Comparison Type=::NewEnumerator7
    ))
  ]
  rendererClasses [&NiagaraMeshRendererProperties]
  renderers [
   {rendererClass=&NiagaraMeshRendererProperties, rendererIndex=0}
   {rendererClass=&NiagaraMeshRendererProperties, rendererIndex=1}
   {rendererClass=&NiagaraMeshRendererProperties, rendererIndex=2}
  ]
 }
 Wave {
  emitterUpdateScript [&EmitterState(Life Cycle Mode=::NewEnumerator1, Loop Behavior=::NewEnumerator1, Loop Duration=4, UseLoopDelay=false)]
  particleSpawnScript [
   &InitializeParticle(
    Color Maximum=<rgba 0.6 0.6667 0.8 1>
    Color Minimum=<rgba 0.75 0.875 1 1>
    Color Mode=::NewEnumerator2
    Lifetime Max=1.7
    Lifetime Min=1.7
    Lifetime Mode=::NewEnumerator1
    Mesh Scale=&RandomRangeVector(Minimum=<0.13 0.1 0.1>, Maximum=<0.09 0.1 0.1>)
    Mesh Scale Mode=::NewEnumerator4
    Sprite Size Mode=::NewEnumerator3
    Uniform Sprite Size=5
    UsePositionOffset=false
    Write Material Random=true
   )
   &InitialMeshOrientation(
    Mesh Axis To Orient=<0 1 0>
    Mesh Orientation Mode=::NewEnumerator2
    Rotation=&RandomRangeVector(Minimum=<-1 0 0>, Maximum=<1 0 0>)
    Use Rotation Vector=true
   )
   &FN_SetSpawnGroup(Inclusivity=true, Normalize Spawn Group=true, Particles Per Group=5, Pattern=::NewEnumerator0, Set Ribbon ID and Ribbon Link Order=true)
  ]
  particleUpdateScript [
   &ParticleState()
   /SetVariables(Particles.Scale Mesh Shape=DataInterface() {propertyValues {Curve {keys [<at 0 0.65>,<at 0.5 1.5>,<at 1 0.65>]}}}) {isSetParameters=true}
   /SetVariables(Particles.Scale Mesh YZ=DataInterface() {propertyValues {Curve {keys [
        {arriveTangent=1.0464, interpMode=RCIM_Cubic, leaveTangent=1.0464, tangentMode=RCTM_User, time=0, value=0}
        {arriveTangent=1.6652, interpMode=RCIM_Cubic, leaveTangent=1.6652, tangentMode=RCTM_User, time=1, value=1}
       ]}}}) {isSetParameters=true}
   &ScaleMeshSize(Scale Factor=&MakeVector(
     X=1
     Y=&FloatFromCurve(
      FloatCurve=$Particles.Scale Mesh YZ
      Scale Curve=&FloatFromCurve(FloatCurve=$Particles.Scale Mesh Shape, CurveIndex=$Particles.Normalized Spawn Group)
     )
     Z=&FloatFromCurve(
      FloatCurve=$Particles.Scale Mesh YZ
      Scale Curve=&FloatFromCurve(FloatCurve=$Particles.Scale Mesh Shape, CurveIndex=$Particles.Normalized Spawn Group)
     )
    ))
   &ScaleColor(
    Scale Alpha=&FloatFromCurve(FloatCurve=DataInterface() {propertyValues {Curve {keys [
         {interpMode=RCIM_Cubic, tangentMode=RCTM_User, time=0, value=0}
         {interpMode=RCIM_Cubic, tangentMode=RCTM_User, time=0.15, value=1}
         {arriveTangent=0.1663, interpMode=RCIM_Cubic, leaveTangent=0.1663, tangentMode=RCTM_User, time=0.7507, value=1}
         {arriveTangent=-7.981, interpMode=RCIM_Cubic, leaveTangent=-7.981, tangentMode=RCTM_User, time=1, value=0}
        ]}}})
    Scale RGB=&VectorFromFloat(Value=&FloatFromCurve(FloatCurve=DataInterface() {propertyValues {Curve {keys [<at 0 1.4 RCIM_Cubic>, {interpMode=RCIM_Cubic, leaveTangent=-0.6944, time=0.35, value=0.75}, <at 1 0.5>]}}}))
    ScaleA=true
    ScaleRGB=true
   )
  ]
  rendererClasses [&NiagaraMeshRendererProperties]
  renderers [
   {rendererClass=&NiagaraMeshRendererProperties, rendererIndex=0}
   {rendererClass=&NiagaraMeshRendererProperties, rendererIndex=1}
  ]
 }
}
summary {
 systemName=NS_Emote_Capsulewhoa_V2_NewShader
 emitters [
  {emitterName=Bullet, rendererClasses [&NiagaraMeshRendererProperties]}
  {emitterName=Wave, rendererClasses [&NiagaraMeshRendererProperties]}
  {emitterName=Empty, rendererClasses [&NiagaraMeshRendererProperties]}
  {emitterName=Bullet001, rendererClasses [&NiagaraMeshRendererProperties]}
 ]
 userVariables []
}
